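# M_Glitter is a Default Lit MaterialGym material that applies the MF_Glitter material function to a plain black base.
# A zero constant sets Base Color, Metallic and Specular through SetMaterialAttributes, which feeds the function's Material input.
# The T_WhitesRGB texture's RGB drives the function's TechArtMask input, and the glitter logic itself lives inside MF_Glitter.
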Material /Game/Global/Materials/MaterialGym/Materials/Glitter/M_Glitter
graph {
 properties {bAutomaticallySetUsageInEditor=false, bUseMaterialAttributes=true, usedShadingModels=Default Lit}
 nodes [
  n1 = Constant()
  MaterialFunctionCall!(
   Material=SetMaterialAttributes!(MaterialAttributes, Base Color=@n1, Metallic=@n1, Specular=@n1) {attributeSetTypes [#1, #2, #3]}
   TechArtMask=TextureSample() {texture=&T_WhitesRGB}.RGB
  ) {materialFunction=&/MF_Glitter}
 ]
}
